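# Niagara system for the candle on the gnome backpack cosmetic: an orange flickering flare with sprite and light renderers.
# A single Flare emitter bursts two particles that loop their lifetime rather than being killed, with the continuous spawn rate disabled.
# An offset waveform drives alpha flicker through a front-end/gameplay switch.
# The front-end branch also multiplies the flicker by normalized system Timeline time, ramping it over the timeline.
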NiagaraSystem /BRCosmetics/Accessories/FORT_Backpacks/Backpack_M_MED_Gnome/FX/NS_Backpack_Gnome_Candle
dependencies {usedDataInterfaces [], usedDynamicInputs [], usedModules [&SystemState, &Timeline]}
emitters {Flare {
  emitterUpdateScript [
   &EmitterState(Life Cycle Mode=::NewEnumerator1, Loop Behavior=::NewEnumerator1, Loop Duration Mode=::NewEnumerator1, UseLoopDelay=false)
   &SpawnBurst_Instantaneous(Spawn Count=2, Use Spawn Probability=false, UseLoopCountLimit=false)
   &SpawnRate(SpawnRate=3, Use Spawn Probability=false) {enabled=false}
  ]
  particleSpawnScript [&InitializeParticle(
    Color=<rgba 1 0.2026 0 1>
    Lifetime Max=0.6
    Lifetime Min=0.4
    Lifetime Mode=::NewEnumerator1
    Sprite Rotation Mode=::NewEnumerator2
    Sprite Size Mode=::NewEnumerator3
    Uniform Sprite Size=50
    UsePositionOffset=false
    Write Material Random=true
   ) {module=InitializeParticle001}]
  particleUpdateScript [
   &FN_Light_Attributes(Create Exponent Attribute=true, Exponent=4)
   &ScaleColor(
    Scale Alpha=&FN_FrontEndFloatModifier(
     Gameplay=&Add_Float(A=0.5, B=&Waveform("Waveform [1]"=::NewEnumerator9, "[1] Frequency"=8, "[1] Amplitude Min/Max"=<0 1>))
     Front End=&Multiply_Float(
      A=&Add_Float(A=0.5, B=&Waveform("Waveform [1]"=::NewEnumerator9, "[1] Frequency"=8, "[1] Amplitude Min/Max"=<0 1>))
      B=$System.Timeline.NormalizedTime
     )
    )
    ScaleA=true
    ScaleRGB=true
   ) {module=ScaleColor001}
   &ParticleState(
    Kill Particles When Lifetime Has Elapsed=false
    Let Infinitely Lived Particles Die When Emitter Deactivates=true
    Loop Particles Lifetime=true
   )
  ]
  rendererClasses [&NiagaraSpriteRendererProperties, &NiagaraLightRendererProperties]
  renderers [
   {rendererClass=&NiagaraSpriteRendererProperties}
   {rendererClass=&NiagaraLightRendererProperties, rendererIndex=1}
  ]
 }}
summary {
 systemName=NS_Backpack_Gnome_Candle
 emitters [{emitterName=Flare, rendererClasses [&NiagaraSpriteRendererProperties, &NiagaraLightRendererProperties]}]
 userVariables [
  User.bIsFrontEnd {type=NiagaraBool}
  User.bIsFrontEndPreview {type=NiagaraBool}
  User.FrontEndToggle {type=NiagaraBool}
 ]
}
