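# Physics asset for the gnome backpack cosmetic mesh, simulating the dangling motion of its lantern.
# It holds two box bodies, one on the lantern root bone and one on the dynamic handle bone.
# A single constraint joins them with tight swing1 and twist limits and a wide swing2 limit, so movement is mostly a hinge-like swing around one axis.
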PhysicsAsset /BRCosmetics/Accessories/FORT_Backpacks/Backpack_M_MED_Gnome/Meshes/M_MED_Gnome_Physics
bodies {
 dyn_handle [{shapeType=Box, extentX=3.75, extentY=18.75, extentZ=5.25, center=<16.3558 0 0.516>}]
 lantern_root [{shapeType=Box, extentX=11.75, extentY=5.75, extentZ=5.5, center=<0 -7.0116 4.5003>}]
}
constraints [{bone1Name=dyn_handle, bone2Name=lantern_root, swing1LimitDegrees=2.5, swing2LimitDegrees=60, twistLimitDegrees=2.5}]
